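# Translucent unlit two-sided material for the placed Creative Fire Volume device, showing animated fire that shifts color and slowly reverses as particle parameter Extinguish rises.
# Three panning samples of one mask, projected in world space by dominant normal axis, blend two colors and drive opacity; low quality uses one.
# UV, view-angle and distance masks shape opacity, while near-camera fade scales emissive.
# MF_CreativeMoveTool processes the final attributes.
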Material /Game/Creative/Devices/FireVolume/Materials/M_CP_FireVolume_Placed_01
graph {
 properties {
  allowTranslucentCustomDepthWrites=true
  bEnableResponsiveAA=true
  bUseMaterialAttributes=true
  bUsedWithMeshParticles=true
  bUsedWithNiagaraMeshParticles=true
  blendMode=BLEND_Translucent
  pixelDepthOffsetMode=PDOM_Legacy
  previewMesh=/Game/Creative/Devices/FireVolume/SM_Cube_FireVolume
  shadingModel=MSM_Unlit
  twoSided=true
  usedShadingModels=Unlit
 }
 nodes [
  n1 = DynamicParameter() {defaultValue=<rgba 0 0 0 1>, paramNames [Extinguish, "", "", ""]}
  n2 = LinearInterpolate(
   A=VectorParameter() {parameterName=BGColor, defaultValue=<rgba 2 0 0.0333 0>}.RGB
   B=VectorParameter() {parameterName=BGColor - Extinguish, defaultValue=<rgba 0.2605 0.1738 2 1>}.RGB
   Alpha=@n1.Extinguish
  )
  n3 = LinearInterpolate(
   A=VectorParameter() {parameterName=Circles, defaultValue=<rgba 5 0.5042 0.1225 0>}.RGB
   B=VectorParameter() {parameterName=Circles - Extinguish, defaultValue=<rgba 0.3872 3.0466 5 1>}.RGB
   Alpha=@n1.Extinguish
  )
  n4 = MaterialFunctionCall(Float3=Multiply(A=WorldPosition().XYZ) {constB=-1}) {materialFunction=&BreakOutFloat3Components}
  n5 = VertexNormalWS()
  n6 = MaterialFunctionCall(Float3=Abs(None=Round(None=@n5))) {materialFunction=&BreakOutFloat3Components}
  n7 = LinearInterpolate(A=AppendVector(A=@n4.G, B=@n4.G), B=AppendVector(A=@n4.R, B=@n4.R), Alpha=@n6.G)
  n8 = AppendVector(A=@n4.R, B=@n4.R)
  n9 = Multiply(
   A=Divide(A=LinearInterpolate(A=@n7, B=@n8, Alpha=@n6.B)) {constB=100}
   B=AppendVector(A=ScalarParameter() {defaultValue=0.8, parameterName=TileU}, B=ScalarParameter() {defaultValue=0.5, parameterName=TileV})
  )
  n10 = Multiply(
   A=Time()
   B=LinearInterpolate(Alpha=DynamicParameter() {defaultValue=<rgba 0 0 0 1>, paramNames [Extinguish, "", "", ""]}.Extinguish) {constA=1, constB=-0.2}
  )
  n11 = ScalarParameter() {defaultValue=0.5, parameterName=Tile Speed 1}
  n12 = TextureObject() {samplerType=SAMPLERTYPE_LinearGrayscale, texture=/Game/Creative/Devices/FireVolume/Materials/Textures/T_FireVolume_M}
  n13 = TextureSample(UVs=Panner(Coordinate=@n9, Time=Multiply(A=@n10, B=@n11)) {speedX=-0.75, speedY=2.5}, Tex=@n12)
  n14 = ScalarParameter() {defaultValue=0.8, parameterName=Tile Scale 2}
  n15 = ScalarParameter() {defaultValue=0.2, parameterName=Tile Speed 2}
  n16 = TextureSample(UVs=Panner(Coordinate=Multiply(A=@n9, B=@n14) {constB=0.7}, Time=Multiply(A=@n10, B=@n15)) {speedX=0.4, speedY=3.25}, Tex=@n12)
  n17 = ScalarParameter() {defaultValue=0.8, parameterName=Tile Scale 3}
  n18 = ScalarParameter() {defaultValue=0.2, parameterName=Tile Speed 2}
  n19 = TextureSample(UVs=Panner(Coordinate=Multiply(A=@n9, B=@n17), Time=Multiply(A=@n10, B=@n18)) {speedX=0.1, speedY=2}, Tex=@n12)
  n20 = Add(A=Multiply(A=Multiply(A=@n13.R, B=@n16.R)) {constB=4}, B=Multiply(A=@n19.R) {constB=0.5})
  n21 = Power(Base=QualitySwitch(Default=@n20, Low=@n19.R, High=@n20, Medium=@n20, Epic=@n20))
  n22 = ParticleColor()
  n23 = Multiply(
   A=Multiply(A=ScalarParameter() {defaultValue=10, parameterName=Intensity Multiplier}, B=LinearInterpolate(A=@n2, B=@n3, Alpha=@n21))
   B=@n22.RGB
  )
  n24 = MaterialFunctionCall(Distance=ScalarParameter() {defaultValue=200, parameterName="Fade_Start_Distance (S)"}) {materialFunction=&Near_Camera_Fade}
  n25 = LinearInterpolate(B=@n21, Alpha=ScalarParameter() {defaultValue=0.99, parameterName=Texture Influence on Opacity}) {constA=1}
  n26 = ComponentMask(None=TextureCoordinate()) {g=true}
  n27 = Power(Base=@n26, Exp=Constant() {r=5}) {constExponent=10}
  n28 = Power(Base=OneMinus(None=@n26), Exp=ScalarParameter() {defaultValue=5, parameterName=Central Contast}) {constExponent=10}
  n29 = Power(Base=OneMinus(None=LinearInterpolate(A=@n27, B=@n28)), Exp=Constant() {r=20}) {constExponent=10}
  n30 = ScalarParameter() {defaultValue=3, parameterName=Fresnel Exponent}
  n31 = Multiply(A=ComponentMask(None=TextureCoordinate()) {r=true}) {constB=4}
  n32 = Sine(None=Multiply(A=Frac(None=@n31)) {constB=0.5})
  n33 = OneMinus(None=Power(Base=OneMinus(None=@n32)) {constExponent=3})
  n34 = Normalize(VectorInput=Subtract(A=CameraPositionWS(), B=ObjectPositionWS()))
  n35 = CameraVectorWS()
  n36 = Power(Base=Abs(None=DotProduct(A=@n34, B=@n35))) {constExponent=20}
  n37 = Multiply(
   A=OneMinus(None=Fresnel(ExponentIn=@n30) {baseReflectFraction=0})
   B=LinearInterpolate(B=Max(A=@n33, B=@n36), Alpha=ScalarParameter() {defaultValue=1, parameterName=Edge Clear}) {constA=1}
  )
  n38 = ShadingPathSwitch(Default=@n37, Deferred=@n37, Mobile=Constant() {r=1}) {desc=Migrated from FeatureLevelSwitch}
  n39 = Constant() {r=1}
  n40 = Multiply(A=@n22.A, B=Multiply(A=Multiply(A=@n25, B=@n29), B=QualitySwitch(Default=@n38, Low=@n39, High=@n38, Medium=@n39, Epic=@n38)))
  n41 = Add(A=Distance(A=CameraPositionWS(), B=WorldPosition().XYZ), B=ScalarParameter() {defaultValue=-100, parameterName=DistanceBias})
  n42 = ScalarParameter() {defaultValue=2000, parameterName=FadeScale}
  n43 = OneMinus(None=Saturate(None=Divide(A=@n41, B=@n42)))
  MaterialFunctionCall(MA In=SetMaterialAttributes!(MaterialAttributes, Emissive Color=Multiply(A=@n23, B=@n24.Result), Opacity=Multiply(A=@n40, B=@n43)) {attributeSetTypes [#1, #2]}) {materialFunction=&MF_CreativeMoveTool}
 ]
}
